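# Editor utility overlay for EUW_GraphDisplay that lets a left-mouse drag select a horizontal range to zoom.
# Captured drags convert cursor X to clamped width fractions, stored and fed to the selection material's StartPercent and EndPercent.
# A Zoom/Clear panel follows the selection midpoint, aligned top or bottom according to the parent graph's value there.
# Zoom broadcasts OnZoomCommitted with the range, then clears; Clear resets and hides the overlay via opacity.
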EditorUtilityWidgetBlueprint /Game/NeverCook/EditorUtilities/Generic/GraphWidgets/GraphDisplay/SubWidgets/EUW_GraphDisplaySelectionOverlay
graphs [
 ClearSelection {nodes [
   ClearSelection() {is=title K2Node_FunctionEntry}
   > Set IsDragging(IsDragging:bool=false) -> (Output_Get:bool)
   > n1 = Set SelectionStartPercent(SelectionStartPercent:float=-1) -> (Output_Get:float)
   > Set SelectionEndPercent(SelectionEndPercent:float=@n1) -> (Output_Get:float)
   > SetRenderOpacity(InOpacity=0)
  ]}
 EventGraph {nodes [
   "On Clicked (btnClear)"!() -> (OutputDelegate, then, widget:User Widget Object Reference, event:Pointer Event Structure) {is=title K2Node_ComponentBoundEvent}
   > ClearSelection() {is=title K2Node_CallFunction}
   "On Clicked (btnZoom)"!() -> (OutputDelegate, then, widget:User Widget Object Reference, event:Pointer Event Structure) {is=title K2Node_ComponentBoundEvent}
   > Call On Zoom Committed!(
    execute:Exec
    self:EUW Graph Display Selection Overlay Object Reference
    StartEndPercent:Vector 2D Structure=MakeVector2D(X=Get SelectionStartPercent() -> (SelectionStartPercent:float), Y=Get SelectionEndPercent() -> (SelectionEndPercent:float))
   ) -> (then:Exec) {is=title K2Node_CallDelegate}
   > ClearSelection()
   Event Construct()
   > ClearSelection()
   Event OnFocusLost!() -> (OutputDelegate, then, InFocusEvent:Focus Event Structure) {is=title K2Node_Event}
   > SetDragging!(execute, self, IsDragging:bool=false) -> (then)
  ]}
 GetGraphDisplayWidget {nodes [
   GetGraphDisplayWidget() {is=title K2Node_FunctionEntry}
   > n1 = Is Valid(InputObject=Get ParentGraphDisplayWidget() -> (ParentGraphDisplayWidget:EUW Graph Display Object Reference))
   n2 = FindParentWidgetOfType!(
    execute=@n1.Is Not Valid
    StartingWidget:Widget Object Reference=Self-Reference()
    Type:Widget Class Reference=/Game/NeverCook/EditorUtilities/Generic/GraphWidgets/GraphDisplay/EUW_GraphDisplay.EUW_GraphDisplay_C
   ) -> (then, ReturnValue:EUW Graph Display Object Reference)
   > n3 = Set ParentGraphDisplayWidget(ParentGraphDisplayWidget:EUW Graph Display Object Reference=@n2) -> (Output_Get:EUW Graph Display Object Reference)
   > Return Node!(execute, ParentGraphDisplayWidget:EUW Graph Display Object Reference=@n3)
   Return Node(execute=@n1.Is Valid, ParentGraphDisplayWidget=Get ParentGraphDisplayWidget())
  ]}
 GetLocalMousePosition {nodes [
   GetLocalMousePosition() {is=title K2Node_FunctionEntry}
   > n1 = GetMousePositionOnPlatform!(execute) -> (then, ReturnValue:Vector 2D Structure)
   > Return Node!(
    execute
    ReturnValue:Vector 2D Structure="Absolute to Local (Coordinates)"!(
     Geometry:"Geometry Structure (by ref)"=GetCachedGeometry!(self:Widget Object Reference) -> (ReturnValue:"Geometry Structure (by ref)")
     AbsoluteCoordinate:Vector 2D Structure=@n1
    ) -> (ReturnValue:Vector 2D Structure)
   )
  ]}
 OnMouseButtonDown {nodes [
   n1 = OnMouseButtonDown!() -> (then, MyGeometry:Geometry Structure, MouseEvent:"Pointer Event Structure (by ref)") {is=title K2Node_FunctionEntry}
   n2 = Get IsDragging() -> (IsDragging:bool)
   > Branch(Condition=AND Boolean(
     A=Is Mouse Button Down!(Input:"Pointer Event Structure (by ref)"=@n1.MouseEvent, MouseButton:Key Structure=LeftMouseButton) -> (ReturnValue:bool)
     B=NOT Boolean(A=@n2)
    ))
   > SetDragging(IsDragging=true)
   > UpdateDragMousePosition!(execute, self, isStart:bool=true) -> (then)
   > UpdateDragMousePosition(isStart=false)
   > SetRenderOpacity(InOpacity=1)
   > Return Node(ReturnValue:Event Reply Structure=CaptureMouse!(Reply:"Event Reply Structure (by ref)"=Unhandled(), CapturingWidget:Widget Object Reference=Self-Reference()) -> (ReturnValue:Event Reply Structure))
  ]}
 OnMouseButtonUp {nodes [
   n1 = OnMouseButtonUp!() -> (then, MyGeometry:Geometry Structure, MouseEvent:"Pointer Event Structure (by ref)") {is=title K2Node_FunctionEntry}
   n2 = Is Mouse Button Down(Input=@n1.MouseEvent, MouseButton=LeftMouseButton)
   > Branch(Condition=AND Boolean(A=NOT Boolean(A=@n2), B=Get IsDragging()))
   > SetDragging(IsDragging=false)
   > Return Node(ReturnValue=ReleaseMouseCapture!(Reply:"Event Reply Structure (by ref)"=Handled()) -> (ReturnValue:Event Reply Structure))
  ]}
 OnMouseMove {nodes [
   n1 = OnMouseMove!() -> (then, MyGeometry:Geometry Structure, MouseEvent:"Pointer Event Structure (by ref)") {is=title K2Node_FunctionEntry}
   > Branch(Condition=Get IsDragging())
   > UpdateOnDrag!(execute, self, DragGeo:Geometry Structure=@n1, Input:"Pointer Event Structure (by ref)"=@n1.MouseEvent) -> (then)
   > Return Node(ReturnValue=Handled())
  ]}
 OnZoomCommitted {nodes [OnZoomCommitted!() -> (then, StartEndPercent:Vector 2D Structure) {is=title K2Node_FunctionEntry}]}
 SetDragging {nodes [
   n1 = SetDragging!() -> (then, IsDragging:bool=~) {is=title K2Node_FunctionEntry}
   > Set IsDragging(IsDragging=@n1)
  ]}
 UpdateDragMousePosition {nodes [
   UpdateDragMousePosition!() -> (then, isStart:bool=~) {is=title K2Node_FunctionEntry}
   > n1 = GetLocalMousePosition!(execute, self) -> (then, ReturnValue:Vector 2D Structure) {is=title K2Node_CallFunction}
   > Set l_position(l_position:Vector 2D Structure=@n1) -> (Output_Get:Vector 2D Structure)
   > n2 = Sequence!(execute) -> (then_0, then_1, then_2:Exec, then_3:Exec)
   n3 = GetCachedGeometry()
   n4 = float / float(
    A=BreakVector2D(InVec=Get l_position() -> (l_position:Vector 2D Structure))
    B=BreakVector2D(InVec=GetLocalSize!(Geometry:"Geometry Structure (by ref)"=@n3) -> (ReturnValue:Vector 2D Structure))
   )
   > Set l_value(l_value:float="Clamp (Float)"(Value="Absolute (Float)"(A=@n4), Min=0, Max=1)) -> (Output_Get:float)
   n5 = GetDynamicMaterial(execute=@n2.then_1, self:Border Object Reference=Get BorderMain() -> (BorderMain:Border Object Reference))
   > SetScalarParameterValue(
    self=@n5
    ParameterName=Select(Option 0:Name=EndPercent, Option 1:Name=StartPercent, Index=Get isStart() -> (isStart:bool)) -> (ReturnValue:Name)
    Value=Get l_value() -> (l_value:float)
   )
   n6 = Branch(execute=@n2.then_2, Condition=Get isStart())
   n7 = Get l_value()
   > Set SelectionStartPercent(SelectionStartPercent=@n7)
   Set SelectionEndPercent(execute=@n6.else, SelectionEndPercent=@n7)
   ValidateButtons(execute=@n2.then_3)
  ]}
 UpdateOnDrag {nodes [
   UpdateOnDrag!() -> (then, DragGeo:Geometry Structure, Input:"Pointer Event Structure (by ref)") {is=title K2Node_FunctionEntry}
   > UpdateDragMousePosition(isStart=false) {is=title K2Node_CallFunction}
  ]}
 ValidateButtons {nodes [
   ValidateButtons() {is=title K2Node_FunctionEntry}
   n1 = "Not Equal (Float)"(A=Get SelectionEndPercent())
   n2 = "Not Equal (Float)"(A=Get SelectionStartPercent())
   > SetVisibility(
    self=Get ButtonPanel() -> (ButtonPanel:Scale Box Object Reference)
    InVisibility=Select(Option 0:ESlateVisibility Enum=Collapsed, Option 1:ESlateVisibility Enum=Visible, Index=AND Boolean(A=@n2, B=@n1)) -> (ReturnValue:ESlateVisibility Enum)
   )
   n3 = Get SelectionStartPercent()
   n4 = Get SelectionEndPercent()
   > Set l_average(l_average:double=float / float(A=float + float(A=@n3, B=@n4), B=2)) -> (Output_Get:double)
   n5 = GetCachedGeometry(self=Get ButtonPanel())
   n6 = float / float(A=BreakVector2D(InVec=GetLocalSize(Geometry=@n5)), B=2)
   n7 = float * float(A=Get l_average() -> (l_average:double), B=GetLocalSize!(Geometry=GetCachedGeometry()) -> (ReturnValue_X:double, ReturnValue_Y:double)) {is=title K2Node_PromotableOperator}
   > SetPadding!(
    execute
    self:Overlay Slot Object Reference=SlotAsOverlaySlot(Widget=Get ButtonPanel())
    InPadding:Margin Structure=Make Margin(Left=float - float(A=@n7, B=@n6), Top=0, Right=0, Bottom=0)
   ) -> (then)
   > n8 = GetGraphDisplayWidget!(execute, self) -> (then, ParentGraphDisplayWidget:EUW Graph Display Object Reference) {is=title K2Node_CallFunction}
   > n9 = GetGraphValueFromPercentage!(execute, self:EUW Graph Display Object Reference=@n8, PercentX:double=Get l_average(), FromTotal:bool=false) -> (then, Point:Vector 2D Structure, PercentValues:Vector 2D Structure)
   > n10 = Branch(Condition="float > float"(A=BreakVector2D(InVec=@n9.PercentValues).Y, B=0.5))
   n11 = SlotAsOverlaySlot(Widget=Get ButtonPanel())
   > Set VerticalAlignment!(execute, VerticalAlignment:EVerticalAlignment Enum=VAlign_Bottom, self:Overlay Slot Object Reference=@n11) -> (then, Output_Get:EVerticalAlignment Enum=VAlign_Fill)
   Set VerticalAlignment(execute=@n10.else, VerticalAlignment=VAlign_Top, self=@n11) -> (Output_Get=VAlign_Fill)
   n12 = Get!(self:EUW Graph Renderer Object Reference=Get GraphRenderer!(self:EUW Graph Display Object Reference=@n8) -> (GraphRenderer:EUW Graph Renderer Object Reference)) -> (DisplayMinMaxY_X:double, DisplayMinMaxY_Y:double)
   float / float(A=float + float(A=@n12, B=@n12.DisplayMinMaxY_Y), B=2)
  ]}
]
variables [
 IsDragging {type=boolean}
 OnZoomCommitted {default=()}
 ParentGraphDisplayWidget {type=EUW_GraphDisplay_C, default=null}
 SelectionEndPercent {type=number}
 SelectionStartPercent {type=number}
]
widgets {
 info {rootWidgetClass=&Overlay, widgetCount=6}
 widgets [
  Overlay_0 {widgetClassPath=&Overlay, widget=Overlay_0}
  BorderMain {
   widgetClassPath=&Border
   bIsVariable=true
   parent=Overlay_0
   widget=BorderMain
   properties {background {
     bIsDynamicallyLoaded=false
     drawAs=Image
     imageType=FullColor
     mirroring=NoMirror
     resourceName=None
     resourceObject=/Game/NeverCook/EditorUtilities/Generic/GraphWidgets/GraphDisplay/Materials/M_GraphDisplaySelection
     tiling=NoTile
     imageSize=<32 32>
     margin=<ltrb 0 0 0 0>
     outlineSettings {
      bUseBrushTransparency=false
      roundingType=HalfHeightRadius
      width=0
      color {colorUseRule=UseColor_Specified, specifiedColor=<rgba 0 0 0 0>}
      cornerRadii=<0 0 0 0>
     }
     tintColor {colorUseRule=UseColor_Specified, specifiedColor=<rgba 1 1 1 1>}
     uVRegion {bIsValid=false, max=<0 0>, min=<0 0>}
    }}
   slot {class=OverlaySlot, properties {horizontalAlignment=HAlign_Fill, verticalAlignment=VAlign_Fill}}
  }
  ButtonPanel {
   widgetClassPath=&ScaleBox
   bIsVariable=true
   parent=Overlay_0
   widget=ButtonPanel
   properties {stretch=UserSpecified, userSpecifiedScale=0.528, visibility=Collapsed}
   slot {class=OverlaySlot}
  }
  HorizontalBox_80 {widgetClassPath=&HorizontalBox, parent=ButtonPanel, widget=HorizontalBox_80, slot {class=ScaleBoxSlot}}
  btnZoom {
   widgetClassPath=&TA_Generic01_Button
   parent=HorizontalBox_80
   widget=btnZoom
   properties {label=ZOOM}
   slot {class=HorizontalBoxSlot}
  }
  btnClear {
   widgetClassPath=&TA_Generic01_Button
   parent=HorizontalBox_80
   widget=btnClear
   properties {label=" X ", buttonColor_Default=<rgba 0.4219 0.0297 0 1>, buttonColor_Selected=<rgba 0.4219 0.0297 0 1>}
   slot {class=HorizontalBoxSlot}
  }
 ]
}
