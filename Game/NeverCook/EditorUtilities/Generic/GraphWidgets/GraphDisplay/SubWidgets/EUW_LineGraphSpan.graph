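# Editor utility sub-widget for the line graph display that draws a translucent span line with a small Y auto-size dropdown over it.
# The isX flag orients the line on pre-construct; on construct it collapses the dropdown when true, otherwise fills it from Enum_GraphRenderAutosize and selects the renderer's current mode.
# Selection changes write AutoSizeY back to the graph renderer, reached through a lazily cached parent EUW_GraphDisplay lookup.
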EditorUtilityWidgetBlueprint /Game/NeverCook/EditorUtilities/Generic/GraphWidgets/GraphDisplay/SubWidgets/EUW_LineGraphSpan
graphs [
 EventGraph {nodes [
   n1 = "On Selection Changed (EditorUtilityComboBoxString_81)"!() -> (OutputDelegate, then, SelectedItem:String, SelectionType:ESelectInfo Enum) {is=title K2Node_ComponentBoundEvent}
   > n2 = GetGraphRenderer!(execute, self) -> (then, GraphRenderer:EUW Graph Renderer Object Reference)
   > Is Valid(InputObject=@n2)
   n3 = FindOptionIndex!(
    self:"Combo Box (String) Object Reference"=Get EditorUtilityComboBoxString_81() -> (EditorUtilityComboBoxString_81:Editor Utility Combo Box String Object Reference)
    Option:String=@n1.SelectedItem
   ) -> (ReturnValue:int=0)
   > SetAutoSizeY!(
    execute
    self:EUW Graph Renderer Object Reference=@n2
    NewAutoSizeY:Enum_GraphRenderAutosize Enum=Byte to Enum Enum_GraphRenderAutosize!(Byte:Byte="To Byte (Integer)"!(InInt:int=@n3) -> (ReturnValue:Byte=0)) -> (ReturnValue:Enum_GraphRenderAutosize Enum) {is=title K2Node_CastByteToEnum}
   ) -> (then)
   Event PreConstruct()
   > ValidateSpanImage()
   Event Construct()
   > InitCombobox()
  ]}
 GetGraphDisplayWidget {nodes [
   GetGraphDisplayWidget() {is=title K2Node_FunctionEntry}
   > n1 = Is Valid(InputObject=Get ParentGraphDisplayWidget() -> (ParentGraphDisplayWidget:EUW Graph Display Object Reference))
   n2 = FindParentWidgetOfType!(
    execute=@n1.Is Not Valid
    StartingWidget:Widget Object Reference=Self-Reference()
    Type:Widget Class Reference=/Game/NeverCook/EditorUtilities/Generic/GraphWidgets/GraphDisplay/EUW_GraphDisplay.EUW_GraphDisplay_C
   ) -> (then, ReturnValue:EUW Graph Display Object Reference)
   > n3 = Set ParentGraphDisplayWidget(ParentGraphDisplayWidget:EUW Graph Display Object Reference=@n2) -> (Output_Get:EUW Graph Display Object Reference)
   > Return Node!(execute, ParentGraphDisplayWidget:EUW Graph Display Object Reference=@n3)
   Return Node(execute=@n1.Is Valid, ParentGraphDisplayWidget=Get ParentGraphDisplayWidget())
  ]}
 GetGraphRenderer {nodes [
   GetGraphRenderer() {is=title K2Node_FunctionEntry}
   > n1 = GetGraphDisplayWidget!(execute, self) -> (then, ParentGraphDisplayWidget:EUW Graph Display Object Reference) {is=title K2Node_CallFunction}
   > Return Node!(
    execute
    GraphRenderer:EUW Graph Renderer Object Reference=Get GraphRenderer!(self:EUW Graph Display Object Reference=@n1) -> (GraphRenderer:EUW Graph Renderer Object Reference)
   )
  ]}
 InitCombobox {nodes [
   InitCombobox() {is=title K2Node_FunctionEntry}
   > n1 = Branch(Condition=Get isX() -> (isX:bool))
   > SetVisibility(self=Get EditorUtilityComboBoxString_81(), InVisibility=Collapsed)
   n2 = ForEach Enum_GraphRenderAutosize!(execute:Exec=@n1.else, SkipHidden:bool) -> (LoopBody:Exec, EnumValue:Enum_GraphRenderAutosize Enum, then:Exec) {is=title K2Node_ForEachElementInEnum}
   n3 = GetGraphRenderer(execute=@n2.then) {is=title K2Node_CallFunction}
   > n4 = Is Valid(InputObject=@n3)
   n5 = Get EditorUtilityComboBoxString_81()
   SetSelectedIndex!(execute=@n4.Is Not Valid, self:"Combo Box (String) Object Reference"=@n5, Index:int=0) -> (then)
   SetSelectedIndex(
    execute=@n4.Is Valid
    self=@n5
    Index="To Integer (Byte)"!(InByte:Byte=Get AutoSizeY!(self:EUW Graph Renderer Object Reference=@n3) -> (AutoSizeY:Enum_GraphRenderAutosize Enum=NewEnumerator0)) -> (ReturnValue:int=0)
   )
   AddOption!(
    execute=@n2.LoopBody
    self:"Combo Box (String) Object Reference"=Get EditorUtilityComboBoxString_81()
    Option:String=Enum to String!(Enumerator:Enum_GraphRenderAutosize Enum=@n2) -> (ReturnValue:String)
   ) -> (then)
  ]}
 ValidateSpanImage {nodes [
   ValidateSpanImage() {is=title K2Node_FunctionEntry}
   n1 = SlotAsOverlaySlot(Widget=Get ImageSpan() -> (ImageSpan:Image Object Reference))
   > SetHorizontalAlignment(
    self=@n1
    InHorizontalAlignment=Select(Option 0:EHorizontalAlignment Enum=HAlign_Center, Option 1:EHorizontalAlignment Enum=HAlign_Fill, Index=Get isX()) -> (ReturnValue:EHorizontalAlignment Enum)
   )
   > SetVerticalAlignment(
    self=@n1
    InVerticalAlignment=Select(Option 0:EVerticalAlignment Enum=VAlign_Fill, Option 1:EVerticalAlignment Enum=VAlign_Center, Index=Get isX()) -> (ReturnValue:EVerticalAlignment Enum)
   )
  ]}
]
variables [
 isX {type=boolean}
 ParentGraphDisplayWidget {type=EUW_GraphDisplay_C, default=null}
 Value {type=number}
]
widgets {
 info {rootWidgetClass=&Overlay, widgetCount=4}
 widgets [
  Overlay_0 {widgetClassPath=&Overlay, widget=Overlay_0}
  ImageSpan {
   widgetClassPath=&Image
   bIsVariable=true
   parent=Overlay_0
   widget=ImageSpan
   properties {
    brush {
     bIsDynamicallyLoaded=false
     drawAs=Image
     imageType=NoImage
     mirroring=NoMirror
     resourceName=None
     resourceObject=null
     tiling=NoTile
     imageSize=<2 2>
     margin=<ltrb 0 0 0 0>
     outlineSettings {
      bUseBrushTransparency=false
      roundingType=HalfHeightRadius
      width=0
      color {colorUseRule=UseColor_Specified, specifiedColor=<rgba 0 0 0 0>}
      cornerRadii=<0 0 0 0>
     }
     tintColor {colorUseRule=UseColor_Specified, specifiedColor=<rgba 1 1 1 1>}
     uVRegion {bIsValid=false, max=<0 0>, min=<0 0>}
    }
    colorAndOpacity=<rgba 1 1 1 0.4261>
   }
   slot {class=OverlaySlot, properties {horizontalAlignment=HAlign_Center, verticalAlignment=VAlign_Fill}}
  }
  SizeBox_0 {
   widgetClassPath=&SizeBox
   parent=Overlay_0
   widget=SizeBox_0
   properties {bOverride_HeightOverride=true, bOverride_WidthOverride=true, heightOverride=20, widthOverride=30}
   slot {class=OverlaySlot, properties {horizontalAlignment=HAlign_Center, verticalAlignment=VAlign_Center}}
  }
  EditorUtilityComboBoxString_81 {
   widgetClassPath=/Script/Blutility.EditorUtilityComboBoxString
   bIsVariable=true
   parent=SizeBox_0
   widget=EditorUtilityComboBoxString_81
   properties {foregroundColor {colorUseRule=UseColor_Specified, specifiedColor=<rgba 0.5271 0.5271 0.5271 1>}}
   slot {class=SizeBoxSlot}
  }
 ]
}
