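# Editor utility sub-widget for a line graph display: an editable numeric box with a reset button for one axis bound, chosen by IsX and IsMin.
# Numeric input goes through SetValue, which refreshes text and tooltips and fires OnValueChanged; non-numeric input restores the stored value.
# Reset and tooltips read the bound from the parent EUW_GraphDisplay's renderer, via a cached parent lookup.
# HandleGraphDataUpdated only checks renderer validity, with no further action.
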EditorUtilityWidgetBlueprint /Game/NeverCook/EditorUtilities/Generic/GraphWidgets/GraphDisplay/SubWidgets/EUW_LineGraphValueDisplay
graphs [
 EventGraph {nodes [
   n1 = "On Text Committed (EditableTextBox_45)"!() -> (OutputDelegate, then, Text:"Text (by ref)", CommitMethod:ETextCommit Enum=Default) {is=title K2Node_ComponentBoundEvent}
   n2 = "To String (Text)"(InText=@n1.Text)
   > n3 = Branch(Condition=IsNumeric!(SourceString:String=@n2) -> (ReturnValue:bool))
   > SetValue!(execute, self, Value:double=String To Float(InString=@n2), ThrowEvent:bool=true, ForceSet:bool=false) -> (then)
   SetValue(execute=@n3.else, Value=Get Value() -> (Value:double), ThrowEvent=false, ForceSet=true)
   "On Clicked (btnReset)"!() -> (OutputDelegate, then, widget:User Widget Object Reference, event:Pointer Event Structure) {is=title K2Node_ComponentBoundEvent}
   > ResetToDefault()
   n4 = Event PreConstruct()
   > Branch(Condition=@n4.IsDesignTime)
   > ValidateVisuals()
   Event Construct()
   > n5 = GetGraphDisplayWidget!(execute, self) -> (then, ParentGraphDisplayWidget:EUW Graph Display Object Reference)
   > Bind Event to On Data Changed(self:EUW Graph Display Object Reference=@n5, Delegate=Create Event()) {is=title K2Node_AddDelegate}
  ]}
 GetGraphDefaultValue {nodes [
   GetGraphDefaultValue() {is=title K2Node_FunctionEntry}
   > n1 = GetGraphRenderer!(execute, self) -> (then, GraphRenderer:EUW Graph Renderer Object Reference)
   > n2 = Is Valid(InputObject=@n1)
   Return Node!(execute=@n2.Is Not Valid, ReturnValue:double=0)
   n3 = GetCurrentMinMaxY!(self:EUW Graph Renderer Object Reference=@n1) -> (CurrentMinMaxY_X:double, CurrentMinMaxY_Y:double)
   n4 = Get IsMin() -> (IsMin:bool)
   n5 = GetCurrentMinMaxX!(self:EUW Graph Renderer Object Reference=@n1) -> (CurrentMinMaxX_X:double, CurrentMinMaxX_Y:double)
   > Return Node(ReturnValue=Select(
     Option 0:double=Select(Option 0=@n3.CurrentMinMaxY_Y, Option 1=@n3, Index=@n4)
     Option 1:double=Select(Option 0=@n5.CurrentMinMaxX_Y, Option 1=@n5, Index=@n4)
     Index=Get IsX() -> (IsX:bool)
    ) -> (ReturnValue:double))
  ]}
 GetGraphDisplayWidget {nodes [
   GetGraphDisplayWidget() {is=title K2Node_FunctionEntry}
   > n1 = Is Valid(InputObject=Get ParentGraphDisplayWidget() -> (ParentGraphDisplayWidget:EUW Graph Display Object Reference))
   n2 = FindParentWidgetOfType!(
    execute=@n1.Is Not Valid
    StartingWidget:Widget Object Reference=Self-Reference()
    Type:Widget Class Reference=/Game/NeverCook/EditorUtilities/Generic/GraphWidgets/GraphDisplay/EUW_GraphDisplay.EUW_GraphDisplay_C
   ) -> (then, ReturnValue:EUW Graph Display Object Reference)
   > n3 = Set ParentGraphDisplayWidget(ParentGraphDisplayWidget:EUW Graph Display Object Reference=@n2) -> (Output_Get:EUW Graph Display Object Reference)
   > Return Node!(execute, ParentGraphDisplayWidget:EUW Graph Display Object Reference=@n3)
   Return Node(execute=@n1.Is Valid, ParentGraphDisplayWidget=Get ParentGraphDisplayWidget())
  ]}
 GetGraphRenderer {nodes [
   GetGraphRenderer() {is=title K2Node_FunctionEntry}
   > n1 = GetGraphDisplayWidget() {is=title K2Node_CallFunction}
   > Return Node!(
    execute
    GraphRenderer:EUW Graph Renderer Object Reference=Get GraphRenderer!(self:EUW Graph Display Object Reference=@n1) -> (GraphRenderer:EUW Graph Renderer Object Reference)
   )
  ]}
 GetValue {nodes [
   GetValue() {is=title K2Node_FunctionEntry}
   > Return Node!(execute, Value:double=Get Value())
  ]}
 HandleGraphDataUpdated {nodes [
   n1 = HandleGraphDataUpdated!() -> (then, GraphDisplay:EUW Graph Display Object Reference) {is=title K2Node_FunctionEntry}
   > Is Valid(InputObject=Get GraphRenderer(self=@n1))
  ]}
 OnValueChanged {nodes [OnValueChanged!() -> (then, OldValue:double=~, NewValue:double=~, Widget:EUW Line Graph Value Display Object Reference) {is=title K2Node_FunctionEntry}]}
 ResetToDefault {nodes [
   ResetToDefault() {is=title K2Node_FunctionEntry}
   > n1 = GetGraphRenderer() {is=title K2Node_CallFunction}
   > Is Valid(InputObject=@n1)
   > n2 = GetGraphDefaultValue!(execute, self) -> (then, ReturnValue:double) {is=title K2Node_CallFunction}
   > SetValue(Value=@n2, ThrowEvent=true, ForceSet=false)
   n3 = GetCurrentMinMaxY(self=@n1)
   n4 = Get IsMin()
   n5 = GetCurrentMinMaxX(self=@n1)
   Select(
    Option 0=Select(Option 0=@n3.CurrentMinMaxY_Y, Option 1=@n3, Index=@n4)
    Option 1=Select(Option 0=@n5.CurrentMinMaxX_Y, Option 1=@n5, Index=@n4)
    Index=Get IsX()
   )
  ]}
 SetValue {nodes [
   n1 = SetValue!() -> (then, Value:double=~, ThrowEvent:bool=true, ForceSet:bool=~) {is=title K2Node_FunctionEntry}
   > Set l_newValue(l_newValue:double=@n1) -> (Output_Get:double)
   > Set l_oldValue(l_oldValue:double=Get Value()) -> (Output_Get:double)
   > n2 = Sequence()
   n3 = Get l_newValue() -> (l_newValue:double)
   n4 = Get Value()
   > Branch(Condition=OR Boolean(A="Not Equal (Float)"(A=@n3, B=@n4), B=Get ForceSet() -> (ForceSet:bool)))
   > Set Value(Value:double=Get l_newValue()) -> (Output_Get:double)
   > ValidateVisuals()
   > Branch(Condition=Get ThrowEvent() -> (ThrowEvent:bool))
   > Call On Value Changed!(
    execute:Exec
    self:EUW Line Graph Value Display Object Reference
    OldValue:double=Get l_oldValue() -> (l_oldValue:double)
    NewValue:double=Get Value()
    Widget:EUW Line Graph Value Display Object Reference=Self-Reference()
   ) -> (then:Exec) {is=title K2Node_CallDelegate}
   ValidateTooltip(execute=@n2.then_1)
  ]}
 ValidateTooltip {nodes [
   ValidateTooltip() {is=title K2Node_FunctionEntry}
   > n1 = GetGraphRenderer()
   > Is Valid(InputObject=@n1)
   > n2 = Sequence!(execute) -> (then_0, then_1, then_2:Exec)
   > n3 = GetGraphDefaultValue()
   > Set l_default(l_default:double=@n3) -> (Output_Get:double)
   n4 = Get IsMin()
   > Set l_GraphString(l_GraphString:Text=Format Text!(
     Format="LOCTEXT(\"{minMax} is {val}\")"
     minMax:Text=Select(Option 0:Text="LOCTEXT(\"Max\")", Option 1:Text="LOCTEXT(\"Min\")", Index=@n4) -> (ReturnValue:Text)
     val:double=@n3
    ) -> (Result)) -> (Output_Get:Text)
   SetToolTipText!(
    execute=@n2.then_1
    self:Widget Object Reference=Get EditableTextBox_45() -> (EditableTextBox_45:Text Box Object Reference)
    InToolTipText:"Text (by ref)"=Format Text!(Format="INVTEXT(\"{curval}\\r\\n{minMax}\")", curval:double=Get Value(), minMax=Get l_GraphString() -> (l_GraphString:Text)) -> (Result)
   ) -> (then)
   SetToolTipText(
    execute=@n2.then_2
    self=Get btnReset() -> (btnReset:TA Generic 01 Button Object Reference)
    InToolTipText=Format Text!(Format="INVTEXT(\"Reset to {minMax}\")", minMax:double=Get l_default() -> (l_default:double)) -> (Result)
   )
  ]}
 ValidateVisuals {nodes [
   ValidateVisuals() {is=title K2Node_FunctionEntry}
   > "SetText (Text Box)"!(
    execute
    self:Text Box Object Reference=Get EditableTextBox_45()
    InText:Text="To Text (Float)"(Value=Get Value(), MinimumIntegralDigits=1, MaximumIntegralDigits=324, MinimumFractionalDigits=0, MaximumFractionalDigits=2)
   ) -> (then)
  ]}
]
variables [
 IsMin {type=boolean}
 IsX {type=boolean}
 OnValueChanged {default=()}
 ParentGraphDisplayWidget {type=EUW_GraphDisplay_C, default=null}
 Value {type=number}
]
widgets {
 info {rootWidgetClass=&HorizontalBox, widgetCount=4}
 widgets [
  HorizontalBox_0 {widgetClassPath=&HorizontalBox, widget=HorizontalBox_0}
  ScaleBox_100 {
   widgetClassPath=&ScaleBox
   parent=HorizontalBox_0
   widget=ScaleBox_100
   properties {stretchDirection=DownOnly}
   slot {class=HorizontalBoxSlot, properties {size {sizeRule=Fill, value=1}}}
  }
  EditableTextBox_45 {
   widgetClassPath=&EditableTextBox
   bIsVariable=true
   parent=ScaleBox_100
   widget=EditableTextBox_45
   properties {widgetStyle {
     backgroundColor {colorUseRule=UseColor_Specified, specifiedColor=<rgba 1 1 1 1>}
     backgroundImageFocused {
      bIsDynamicallyLoaded=false
      drawAs=RoundedBox
      imageType=NoImage
      mirroring=NoMirror
      resourceName=None
      resourceObject=null
      tiling=NoTile
      imageSize=<0 0>
      margin=<ltrb 0 0 0 0>
      outlineSettings {
       bUseBrushTransparency=false
       roundingType=FixedRadius
       width=1
       color {colorUseRule=UseColor_ColorTable, specifiedColor=<rgba 0 0 0 0>}
       cornerRadii=<4 4 4 4>
      }
      tintColor {colorUseRule=UseColor_Specified, specifiedColor=<rgba 0.0048 0.0048 0.0048 1>}
      uVRegion {bIsValid=false, max=<0 0>, min=<0 0>}
     }
     backgroundImageHovered {
      bIsDynamicallyLoaded=false
      drawAs=RoundedBox
      imageType=NoImage
      mirroring=NoMirror
      resourceName=None
      resourceObject=null
      tiling=NoTile
      imageSize=<0 0>
      margin=<ltrb 0 0 0 0>
      outlineSettings {
       bUseBrushTransparency=false
       roundingType=FixedRadius
       width=1
       color {colorUseRule=UseColor_ColorTable, specifiedColor=<rgba 0 0 0 0>}
       cornerRadii=<4 4 4 4>
      }
      tintColor {colorUseRule=UseColor_Specified, specifiedColor=<rgba 0.0048 0.0048 0.0048 1>}
      uVRegion {bIsValid=false, max=<0 0>, min=<0 0>}
     }
     backgroundImageNormal {
      bIsDynamicallyLoaded=false
      drawAs=RoundedBox
      imageType=NoImage
      mirroring=NoMirror
      resourceName=None
      resourceObject=null
      tiling=NoTile
      imageSize=<0 0>
      margin=<ltrb 0 0 0 0>
      outlineSettings {
       bUseBrushTransparency=false
       roundingType=FixedRadius
       width=1
       color {colorUseRule=UseColor_ColorTable, specifiedColor=<rgba 0 0 0 0>}
       cornerRadii=<4 4 4 4>
      }
      tintColor {colorUseRule=UseColor_Specified, specifiedColor=<rgba 0.0048 0.0048 0.0048 1>}
      uVRegion {bIsValid=false, max=<0 0>, min=<0 0>}
     }
     backgroundImageReadOnly {
      bIsDynamicallyLoaded=false
      drawAs=RoundedBox
      imageType=NoImage
      mirroring=NoMirror
      resourceName=None
      resourceObject=null
      tiling=NoTile
      imageSize=<0 0>
      margin=<ltrb 0 0 0 0>
      outlineSettings {
       bUseBrushTransparency=false
       roundingType=FixedRadius
       width=1
       color {colorUseRule=UseColor_ColorTable, specifiedColor=<rgba 0 0 0 0>}
       cornerRadii=<4 4 4 4>
      }
      tintColor {colorUseRule=UseColor_Specified, specifiedColor=<rgba 0.0284 0.0284 0.0284 1>}
      uVRegion {bIsValid=false, max=<0 0>, min=<0 0>}
     }
     focusedForegroundColor {colorUseRule=UseColor_Specified, specifiedColor=<rgba 1 1 1 1>}
     foregroundColor {colorUseRule=UseColor_Specified, specifiedColor=<rgba 0.5271 0.5271 0.5271 1>}
     hScrollBarPadding=<ltrb 0 0 0 0>
     padding=<ltrb 0 0 0 0>
     readOnlyForegroundColor {colorUseRule=UseColor_Foreground, specifiedColor=<rgba 1 0 1 1>}
     scrollBarStyle {
      thickness=8
      draggedThumbImage {
       bIsDynamicallyLoaded=false
       drawAs=RoundedBox
       imageType=NoImage
       mirroring=NoMirror
       resourceName=None
       resourceObject=null
       tiling=NoTile
       imageSize=<0 0>
       margin=<ltrb 0 0 0 0>
       outlineSettings {
        bUseBrushTransparency=false
        roundingType=FixedRadius
        width=0
        color {colorUseRule=UseColor_Specified, specifiedColor=<rgba 0 0 0 0>}
        cornerRadii=<4 4 4 4>
       }
       tintColor {colorUseRule=UseColor_Specified, specifiedColor=<rgba 0.2159 0.2159 0.2159 1>}
       uVRegion {bIsValid=false, max=<0 0>, min=<0 0>}
      }
      horizontalBackgroundImage {
       bIsDynamicallyLoaded=false
       drawAs=NoDrawType
       imageType=NoImage
       mirroring=NoMirror
       resourceName=None
       resourceObject=null
       tiling=NoTile
       imageSize=<0 0>
       margin=<ltrb 0 0 0 0>
       outlineSettings {
        bUseBrushTransparency=false
        roundingType=HalfHeightRadius
        width=0
        color {colorUseRule=UseColor_Specified, specifiedColor=<rgba 0 0 0 0>}
        cornerRadii=<0 0 0 0>
       }
       tintColor {colorUseRule=UseColor_Specified, specifiedColor=<rgba 1 1 1 1>}
       uVRegion {bIsValid=false, max=<0 0>, min=<0 0>}
      }
      horizontalBottomSlotImage {
       bIsDynamicallyLoaded=false
       drawAs=NoDrawType
       imageType=NoImage
       mirroring=NoMirror
       resourceName=None
       resourceObject=null
       tiling=NoTile
       imageSize=<0 0>
       margin=<ltrb 0 0 0 0>
       outlineSettings {
        bUseBrushTransparency=false
        roundingType=HalfHeightRadius
        width=0
        color {colorUseRule=UseColor_Specified, specifiedColor=<rgba 0 0 0 0>}
        cornerRadii=<0 0 0 0>
       }
       tintColor {colorUseRule=UseColor_Specified, specifiedColor=<rgba 1 1 1 1>}
       uVRegion {bIsValid=false, max=<0 0>, min=<0 0>}
      }
      horizontalTopSlotImage {
       bIsDynamicallyLoaded=false
       drawAs=NoDrawType
       imageType=NoImage
       mirroring=NoMirror
       resourceName=None
       resourceObject=null
       tiling=NoTile
       imageSize=<0 0>
       margin=<ltrb 0 0 0 0>
       outlineSettings {
        bUseBrushTransparency=false
        roundingType=HalfHeightRadius
        width=0
        color {colorUseRule=UseColor_Specified, specifiedColor=<rgba 0 0 0 0>}
        cornerRadii=<0 0 0 0>
       }
       tintColor {colorUseRule=UseColor_Specified, specifiedColor=<rgba 1 1 1 1>}
       uVRegion {bIsValid=false, max=<0 0>, min=<0 0>}
      }
      hoveredThumbImage {
       bIsDynamicallyLoaded=false
       drawAs=RoundedBox
       imageType=NoImage
       mirroring=NoMirror
       resourceName=None
       resourceObject=null
       tiling=NoTile
       imageSize=<0 0>
       margin=<ltrb 0 0 0 0>
       outlineSettings {
        bUseBrushTransparency=false
        roundingType=FixedRadius
        width=0
        color {colorUseRule=UseColor_Specified, specifiedColor=<rgba 0 0 0 0>}
        cornerRadii=<4 4 4 4>
       }
       tintColor {colorUseRule=UseColor_Specified, specifiedColor=<rgba 0.2159 0.2159 0.2159 1>}
       uVRegion {bIsValid=false, max=<0 0>, min=<0 0>}
      }
      normalThumbImage {
       bIsDynamicallyLoaded=false
       drawAs=RoundedBox
       imageType=NoImage
       mirroring=NoMirror
       resourceName=None
       resourceObject=null
       tiling=NoTile
       imageSize=<0 0>
       margin=<ltrb 0 0 0 0>
       outlineSettings {
        bUseBrushTransparency=false
        roundingType=FixedRadius
        width=0
        color {colorUseRule=UseColor_Specified, specifiedColor=<rgba 0 0 0 0>}
        cornerRadii=<4 4 4 4>
       }
       tintColor {colorUseRule=UseColor_Specified, specifiedColor=<rgba 0.0953 0.0953 0.0953 1>}
       uVRegion {bIsValid=false, max=<0 0>, min=<0 0>}
      }
      verticalBackgroundImage {
       bIsDynamicallyLoaded=false
       drawAs=NoDrawType
       imageType=NoImage
       mirroring=NoMirror
       resourceName=None
       resourceObject=null
       tiling=NoTile
       imageSize=<0 0>
       margin=<ltrb 0 0 0 0>
       outlineSettings {
        bUseBrushTransparency=false
        roundingType=HalfHeightRadius
        width=0
        color {colorUseRule=UseColor_Specified, specifiedColor=<rgba 0 0 0 0>}
        cornerRadii=<0 0 0 0>
       }
       tintColor {colorUseRule=UseColor_Specified, specifiedColor=<rgba 1 1 1 1>}
       uVRegion {bIsValid=false, max=<0 0>, min=<0 0>}
      }
      verticalBottomSlotImage {
       bIsDynamicallyLoaded=false
       drawAs=NoDrawType
       imageType=NoImage
       mirroring=NoMirror
       resourceName=None
       resourceObject=null
       tiling=NoTile
       imageSize=<0 0>
       margin=<ltrb 0 0 0 0>
       outlineSettings {
        bUseBrushTransparency=false
        roundingType=HalfHeightRadius
        width=0
        color {colorUseRule=UseColor_Specified, specifiedColor=<rgba 0 0 0 0>}
        cornerRadii=<0 0 0 0>
       }
       tintColor {colorUseRule=UseColor_Specified, specifiedColor=<rgba 1 1 1 1>}
       uVRegion {bIsValid=false, max=<0 0>, min=<0 0>}
      }
      verticalTopSlotImage {
       bIsDynamicallyLoaded=false
       drawAs=NoDrawType
       imageType=NoImage
       mirroring=NoMirror
       resourceName=None
       resourceObject=null
       tiling=NoTile
       imageSize=<0 0>
       margin=<ltrb 0 0 0 0>
       outlineSettings {
        bUseBrushTransparency=false
        roundingType=HalfHeightRadius
        width=0
        color {colorUseRule=UseColor_Specified, specifiedColor=<rgba 0 0 0 0>}
        cornerRadii=<0 0 0 0>
       }
       tintColor {colorUseRule=UseColor_Specified, specifiedColor=<rgba 1 1 1 1>}
       uVRegion {bIsValid=false, max=<0 0>, min=<0 0>}
      }
     }
     textStyle {
      overflowPolicy=Clip
      transformPolicy=None
      colorAndOpacity {colorUseRule=UseColor_Foreground, specifiedColor=<rgba 1 0 1 1>}
      font {
       bForceMonospaced=false
       bMaterialIsStencil=false
       fontMaterial=null
       fontObject=null
       letterSpacing=0
       monospacedWidth=1
       size=9
       skewAmount=0
       typefaceFontName=Regular
       outlineSettings {
        bApplyOutlineToDropShadows=false
        bMiteredCorners=false
        bSeparateFillAlpha=false
        outlineMaterial=null
        outlineSize=0
        outlineColor=<rgba 0 0 0 1>
       }
      }
      highlightColor {colorUseRule=UseColor_ColorTable, specifiedColor=<rgba 0 0 0 0>}
      highlightShape {
       bIsDynamicallyLoaded=false
       drawAs=Image
       imageType=NoImage
       mirroring=NoMirror
       resourceName=None
       resourceObject=null
       tiling=NoTile
       imageSize=<0 0>
       margin=<ltrb 0 0 0 0>
       outlineSettings {
        bUseBrushTransparency=false
        roundingType=HalfHeightRadius
        width=0
        color {colorUseRule=UseColor_Specified, specifiedColor=<rgba 0 0 0 0>}
        cornerRadii=<0 0 0 0>
       }
       tintColor {colorUseRule=UseColor_ColorTable, specifiedColor=<rgba 0 0 0 0>}
       uVRegion {bIsValid=false, max=<0 0>, min=<0 0>}
      }
      selectedBackgroundColor {colorUseRule=UseColor_ColorTable, specifiedColor=<rgba 0 0 0 0>}
      shadowColorAndOpacity=<rgba 0 0 0 1>
      shadowOffset=<0 0>
      strikeBrush {
       bIsDynamicallyLoaded=false
       drawAs=Image
       imageType=NoImage
       mirroring=NoMirror
       resourceName=None
       resourceObject=null
       tiling=NoTile
       imageSize=<32 32>
       margin=<ltrb 0 0 0 0>
       outlineSettings {
        bUseBrushTransparency=false
        roundingType=HalfHeightRadius
        width=0
        color {colorUseRule=UseColor_Specified, specifiedColor=<rgba 0 0 0 0>}
        cornerRadii=<0 0 0 0>
       }
       tintColor {colorUseRule=UseColor_Specified, specifiedColor=<rgba 1 1 1 1>}
       uVRegion {bIsValid=false, max=<0 0>, min=<0 0>}
      }
      underlineBrush {
       bIsDynamicallyLoaded=false
       drawAs=Image
       imageType=NoImage
       mirroring=NoMirror
       resourceName=None
       resourceObject=null
       tiling=NoTile
       imageSize=<32 32>
       margin=<ltrb 0 0 0 0>
       outlineSettings {
        bUseBrushTransparency=false
        roundingType=HalfHeightRadius
        width=0
        color {colorUseRule=UseColor_Specified, specifiedColor=<rgba 0 0 0 0>}
        cornerRadii=<0 0 0 0>
       }
       tintColor {colorUseRule=UseColor_Specified, specifiedColor=<rgba 1 1 1 1>}
       uVRegion {bIsValid=false, max=<0 0>, min=<0 0>}
      }
     }
     vScrollBarPadding=<ltrb 0 0 0 0>
    }}
   slot {class=ScaleBoxSlot, properties {horizontalAlignment=HAlign_Fill, verticalAlignment=VAlign_Top}}
  }
  btnReset {
   widgetClassPath=&TA_Generic01_Button
   parent=HorizontalBox_0
   widget=btnReset
   properties {iconBrush {
     bIsDynamicallyLoaded=false
     drawAs=Image
     imageType=FullColor
     mirroring=NoMirror
     resourceName=None
     resourceObject=/Game/Athena/UI/InGame/Creative/PropertyEditor/Icon_Picker/Textures_V2/64x64/T_UI_IconLibrary_Reset_64
     tiling=NoTile
     imageSize=<8 8>
     margin=<ltrb 0 0 0 0>
     outlineSettings {
      bUseBrushTransparency=false
      roundingType=HalfHeightRadius
      width=0
      color {colorUseRule=UseColor_Specified, specifiedColor=<rgba 0 0 0 0>}
      cornerRadii=<0 0 0 1>
     }
     tintColor {colorUseRule=UseColor_Specified, specifiedColor=<rgba 1 1 1 1>}
     uVRegion {bIsValid=false, max=<0 0>, min=<0 0>}
    }}
   slot {class=HorizontalBoxSlot, properties {verticalAlignment=VAlign_Top}}
  }
 ]
}
